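# Material function for the Patches pickaxe that adds an animated emissive crack glow to incoming material attributes.
# The mask texture's alpha darkens base color and blends emissive where the effect applies.
# Named reroutes carry an adjustable emissive-texture color, a sphere-mask gradient and a crack-center boost from the mask's blue channel.
# Distorted panning noise adds shimmer, while noise panned along radial distance from an offset center drives bright traveling pulses.
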MaterialFunction /BRCosmetics/Weapons/FORT_Melee/Pickaxe_M_MED_Patches/Materials/MF_Patches_Pickaxe_emissive
graph {nodes [
  n1 = GetMaterialAttributes(None=FunctionInput() {inputType=FunctionInput_MaterialAttributes}) {attributeGetTypes [#1]}
  n2 = TextureSample() {samplerSource=SSM_Wrap_WorldGroupSettings, texture=/BRCosmetics/Weapons/FORT_Melee/Pickaxe_M_MED_Patches/Textures/T_Patches_Pickaxe_Mask}
  n3 = NamedRerouteUsage()
  FunctionOutput(None=SetMaterialAttributes!(
    MaterialAttributes=@n1.MaterialAttributes
    BaseColor=LinearInterpolate(A=@n1.BaseColor, Alpha=@n2.A) {constB=0}
    EmissiveColor=LinearInterpolate(B=@n3, Alpha=@n2.A)
   ) {attributeSetTypes [#1, #2]})
  n4 = TextureSampleParameter2D() {
   group=Patches Pickaxe
   parameterName=Emissive color
   samplerSource=SSM_Wrap_WorldGroupSettings
   texture=/BRCosmetics/Weapons/FORT_Melee/Pickaxe_M_MED_Patches/Textures/T_Patches_Pickaxe_E
  }
  n5 = ScalarParameter() {defaultValue=-0.7, group=Patches Pickaxe, parameterName=Desaturation}
  NamedRerouteDeclaration(None=Multiply(A=Desaturation(None=@n4.RGB, Fraction=@n5), B=ScalarParameter() {defaultValue=1, group=Patches Pickaxe, parameterName=Emissive brightness}) {constB=0.6}) {name=Color, nodeColor=<rgba 0.2188 0 1 1>}
  n6 = TextureCoordinate()
  n7 = Constant() {r=16}
  n8 = TextureSample(UVs=Panner(Coordinate=Multiply(A=@n6, B=@n7)) {speedX=0.1}) {
   samplerSource=SSM_Wrap_WorldGroupSettings
   samplerType=SAMPLERTYPE_Grayscale
   texture=/Game/TimeOfDay/TODM_Sequence/Materials/Sky/Textures/LowResBlurredNoise_Radish_LF_Root
  }
  n9 = Add(A=TextureCoordinate(), B=Multiply(A=@n8.R, B=Constant() {r=0.02}))
  n10 = Constant() {r=6}
  n11 = TextureSample(UVs=Panner(Coordinate=Multiply(A=@n9, B=@n10)) {speedX=0.3}) {
   samplerSource=SSM_Wrap_WorldGroupSettings
   samplerType=SAMPLERTYPE_Grayscale
   texture=&LowResBlurredNoise_Radish_LF_Root
  }
  n12 = Multiply(A=Power(Base=@n11.R, Exp=Constant() {r=2}), B=Constant() {r=0.3}) {constB=0.05}
  n13 = NamedRerouteUsage()
  n14 = Multiply(A=NamedRerouteUsage(), B=NamedRerouteUsage()) {constB=1.01}
  n15 = Constant() {r=1}
  n16 = ScalarParameter() {defaultValue=1500, group=Patches Pickaxe, parameterName=Emissive brightness mult}
  n17 = NamedRerouteUsage()
  n18 = NamedRerouteUsage()
  n19 = Multiply(A=Multiply(A=NamedRerouteUsage(), B=LinearInterpolate(A=@n15, B=@n16, Alpha=@n17) {constA=1}), B=Saturate(None=Multiply(A=@n18) {constB=4}))
  n20 = ComponentMask(None=Constant3Vector() {constant=<rgba -0.2147 -0.7628 0 1>}) {g=true, r=true}
  n21 = MaterialFunctionCall(UVs=TextureCoordinate(), Texture Scale=Constant() {r=0.55}) {materialFunction=&ScaleUVsByCenter}
  n22 = TextureSample(UVs=Multiply(A=TextureCoordinate(), B=Constant() {r=10})) {
   samplerSource=SSM_Wrap_WorldGroupSettings
   samplerType=SAMPLERTYPE_Grayscale
   texture=&LowResBlurredNoise_Radish_LF_Root
  }
  n23 = Constant() {r=0.02}
  n24 = MaterialFunctionCall(Vector or UVs=Add(A=Add(A=@n20, B=@n21.UVs), B=Multiply(A=@n22.R, B=@n23))) {materialFunction=&VectorToRadialValue}
  n25 = AppendVector(A=Constant(), B=Constant() {r=0.4})
  n26 = TextureSample(UVs=Panner(Coordinate=Multiply(A=@n24.Linear Distance, B=@n25)) {speedY=-0.06}) {
   samplerSource=SSM_Wrap_WorldGroupSettings
   samplerType=SAMPLERTYPE_Grayscale
   texture=&LowResBlurredNoise_Radish_LF_Root
  }
  n27 = Constant() {r=8}
  n28 = Multiply(A=NamedRerouteUsage(), B=Saturate(None=Power(Base=@n26.R, Exp=@n27)))
  NamedRerouteDeclaration(None=Add(A=Multiply(A=@n12, B=@n13), B=LinearInterpolate(A=@n14, B=@n19, Alpha=@n28) {constA=100, constB=2000})) {name=Emissive, nodeColor=<rgba 0.0312 0 1 1>}
  n29 = Constant3Vector() {constant=<rgba 0.408 0.696 0 1>}
  n30 = Constant3Vector() {constant=<rgba 0.288 0.568 0 1>}
  n31 = TextureSample() {samplerSource=SSM_Wrap_WorldGroupSettings, texture=&T_Patches_Pickaxe_Mask}
  n32 = LinearInterpolate(
   A=SphereMask(A=TextureCoordinate(), B=ComponentMask(None=@n29) {g=true, r=true}, Radius=Constant() {r=0.42}, Hardness=Constant() {r=0.02})
   B=SphereMask(A=TextureCoordinate(), B=ComponentMask(None=@n30) {g=true, r=true}, Radius=Constant() {r=0.42}, Hardness=Constant())
   Alpha=@n31.G
  )
  NamedRerouteDeclaration(None=Saturate(None=Power(Base=@n32))) {name=Gradient mask, nodeColor=<rgba 0 0.25 1 1>}
  n33 = TextureSample() {samplerSource=SSM_Wrap_WorldGroupSettings, texture=&T_Patches_Pickaxe_Mask}
  n34 = Multiply(A=@n33.B) {constB=5}
  NamedRerouteDeclaration(None=LinearInterpolate(A=Constant() {r=0.5}, B=Constant() {r=1.2}, Alpha=Saturate(None=@n34))) {name=Middle of crack, nodeColor=<rgba 0.8516 1 0 1>}
 ]}
